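# Anim Blueprint for the JadeTowelHope backpack, posing the pack and simulating its strap and celery bones.
# Front-end and in-game chains, blended by bIsInFrontEnd, layer additive bone offsets with spring-based AnimDynamics.
# The front end corrects fit for JadeTowelHope or JadeTowelGloss bodies by mesh-name wildcard, plus Gloss strap rotation from spine twist.
# In game, offsets follow skydiving, sprinting and crouching, dyn_pelvis copies body spine_01, and gravity comes from state-based overrides.
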AnimBlueprint /BRCosmetics/Accessories/FORT_Backpacks/Backpack_F_MED_JadeTowelHope/Meshes/F_MED_JadeTowelHope_Pack_AnimBP
target_skeleton=&/F_MED_JadeTowelHope_Pack_Skeleton
graphs [
 AnimGraph {nodes [
   n1 = Get bIsSkydiveFloating()
   "Transform (Modify) Bone - Bone: pack"!(ComponentPose, bAlphaBoolEnabled:bool=@n1, Translation="0,-2,0", Rotation="-7,0,10", Scale) -> (Pose) {
    alphaInputType=Bool
    rotationMode=BMM_Additive
    rotationSpace=BCS_BoneSpace
    translationMode=BMM_Additive
    translationSpace=BCS_BoneSpace
    alphaBoolBlend {blendInTime=0.5, blendOption=Linear, blendOutTime=0.5}
    boneToModify {boneName=pack}
    rotation {pitch=-7, roll=10}
    translation {y=-2}
   }
   "Transform (Modify) Bone - Bone: pack"(bAlphaBoolEnabled=Get JadeTowelGlossBody() -> (JadeTowelGlossBody:bool), Translation="-3,0,0", Rotation="12,0,12", Scale="0.9,0.9,0.9") {
    alphaInputType=Bool
    rotationMode=BMM_Additive
    rotationSpace=BCS_BoneSpace
    scaleMode=BMM_Additive
    scaleSpace=BCS_BoneSpace
    translationMode=BMM_Additive
    translationSpace=BCS_BoneSpace
    alphaBoolBlend {blendOption=Linear}
    boneToModify {boneName=pack}
    rotation {pitch=12, roll=12}
    scale=<0.9 0.9 0.9>
    translation {x=-3}
   }
   "Transform (Modify) Bone - Bone: pack"(bAlphaBoolEnabled=Get JadeTowelHopeBody() -> (JadeTowelHopeBody:bool), Translation="-2,0,-3.5", Rotation="6,0,11", Scale="0.95,0.95,0.95") {
    alphaInputType=Bool
    rotationMode=BMM_Additive
    rotationSpace=BCS_BoneSpace
    scaleMode=BMM_Additive
    scaleSpace=BCS_BoneSpace
    translationMode=BMM_Additive
    translationSpace=BCS_BoneSpace
    alphaBoolBlend {blendOption=Linear}
    boneToModify {boneName=pack}
    rotation {pitch=6, roll=11}
    scale=<0.95 0.95 0.95>
    translation {x=-2, z=-3.5}
   }
   n2 = "Transform (Modify) Bone - Bone: pack"!(ComponentPose=Local To Component(LocalPose=Use cached pose 'InputPose'()), Alpha:float, Translation="0,1,2", Rotation="0,0,0", Scale) -> (Pose) {translationMode=BMM_Additive, alphaBoolBlend {blendOption=Linear}, boneToModify {boneName=pack}, translation {y=1, z=2}}
   n3 = "Transform (Modify) Bone - Bone: pack"!(ComponentPose=@n2, bAlphaBoolEnabled:bool=Get JadeTowelHopeBody(), Translation="-3,0,-3.5", Rotation="14,2,11", Scale="0.95,0.95,0.95") -> (Pose) {
    alphaInputType=Bool
    rotationMode=BMM_Additive
    rotationSpace=BCS_BoneSpace
    scaleMode=BMM_Additive
    scaleSpace=BCS_BoneSpace
    translationMode=BMM_Additive
    translationSpace=BCS_BoneSpace
    alphaBoolBlend {blendOption=Linear}
    boneToModify {boneName=pack}
    rotation=<rot 14 2 11>
    scale=<0.95 0.95 0.95>
    translation {x=-3, z=-3.5}
   }
   n4 = "Transform (Modify) Bone - Bone: pack"(ComponentPose=@n3, bAlphaBoolEnabled=Get JadeTowelGlossBody(), Translation="-2,0,-3.5", Rotation="14,2,11", Scale="0.9,0.9,0.9") {
    alphaInputType=Bool
    rotationMode=BMM_Additive
    rotationSpace=BCS_BoneSpace
    scaleMode=BMM_Additive
    scaleSpace=BCS_BoneSpace
    translationMode=BMM_Additive
    translationSpace=BCS_BoneSpace
    alphaBoolBlend {blendOption=Linear}
    boneToModify {boneName=pack}
    rotation=<rot 14 2 11>
    scale=<0.9 0.9 0.9>
    translation {x=-2, z=-3.5}
   }
   n5 = Get BodyAnimIdleFrontEnd_Z() -> (BodyAnimIdleFrontEnd_Z:double)
   n6 = "Transform (Modify) Bone - Bone: dyn_strap_base"(ComponentPose=@n4, Alpha=@n5, Translation="0,0,0", Rotation="-0.000,-12.848,-0.000", Scale="1,1,1") {
    rotationMode=BMM_Additive
    rotationSpace=BCS_BoneSpace
    scaleMode=BMM_Additive
    scaleSpace=BCS_BoneSpace
    translationMode=BMM_Additive
    translationSpace=BCS_BoneSpace
    alphaBoolBlend {blendOption=Linear}
    alphaScaleBiasClamp {bInterpResult=true, interpSpeedDecreasing=2}
    boneToModify {boneName=dyn_strap_base}
    rotation=<rot 0 -12.848 0>
   }
   n7 = "Transform (Modify) Bone - Bone: dyn_strap_01"(ComponentPose=@n6, Alpha=@n5, Translation="0,0,0", Rotation="0,-35,0", Scale="1,1,1") {
    rotationMode=BMM_Additive
    rotationSpace=BCS_BoneSpace
    scaleMode=BMM_Additive
    scaleSpace=BCS_BoneSpace
    translationMode=BMM_Additive
    translationSpace=BCS_BoneSpace
    alphaBoolBlend {blendOption=Linear}
    alphaScaleBiasClamp {bInterpResult=true, interpSpeedDecreasing=2}
    boneToModify {boneName=dyn_strap_01}
    rotation {yaw=-35}
   }
   n8 = Get GravityOverride()
   n9 = Anim Dynamics - Bone: dyn_celery_01!(ComponentPose=@n7, Alpha=1, GravityOverride:Vector=@n8, ExternalForce="0,0,0") -> (Pose) {
    angularDampingOverride=0.9999
    angularSpringConstant=500
    bAngularSpring=true
    bOverrideAngularDamping=true
    bUseGravityOverride=true
    bUsePlanarLimit=false
    simulationSpace=RootRelative
    alphaBoolBlend {blendInTime=0.1, blendOption=Linear, blendOutTime=0.1}
    boundBone {boneName=dyn_celery_01}
    componentAppliedLinearAccClamp=<10000 1000 1500>
    componentLinearAccScale=<1 1 1>
    componentLinearVelScale=<1 1 1>
    gravityOverride {z=-400}
    physicsBodyDefinitions [{
      collisionType=CustomSphere
      sphereCollisionRadius=1
      boundBone {boneName=dyn_celery_01}
      boxExtents=<10 10 10>
      constraintSetup {angularLimitsMax=<0 1 6>, angularLimitsMin=<0 -15 -2>, angularTarget=<0 0 0>}
      localJointOffset=<1 0 0>
     }]
   }
   n10 = Anim Dynamics - Bone: dyn_celery_02!(ComponentPose=@n9, Alpha=1, GravityOverride:Vector="0,0,0", ExternalForce="0,0,0") -> (Pose) {
    angularDampingOverride=0.9999
    angularSpringConstant=3000
    bAngularSpring=true
    bOverrideAngularDamping=true
    bUseGravityOverride=true
    bUsePlanarLimit=false
    gravityScale=0.35
    simulationSpace=BoneRelative
    alphaBoolBlend {blendOption=Linear}
    boundBone {boneName=dyn_celery_02}
    componentAppliedLinearAccClamp=<3500 3500 3500>
    componentLinearAccScale=<0.35 0.35 0.35>
    componentLinearVelScale=<0.35 0.35 0.35>
    gravityOverride {z=-700}
    physicsBodyDefinitions [{
      boundBone {boneName=dyn_celery_02}
      boxExtents=<13 13 13>
      constraintSetup {angularLimitsMax=<0 15 0>, angularLimitsMin=<0 -1 0>, angularTarget=<1 0 0>}
      localJointOffset=<1 0 0>
     }]
    relativeSpaceBone {boneName=root}
   }
   n11 = Anim Dynamics - Bone: dyn_strap_00!(ComponentPose=@n10, Alpha=1, GravityOverride:Vector=@n8, ExternalForce="0,0,0") -> (Pose) {
    angularDampingOverride=0.9999
    angularSpringConstant=500
    bAngularSpring=true
    bOverrideAngularDamping=true
    bUseGravityOverride=true
    bUsePlanarLimit=false
    simulationSpace=RootRelative
    alphaBoolBlend {blendInTime=0.1, blendOption=Linear, blendOutTime=0.1}
    boundBone {boneName=dyn_strap_00}
    componentAppliedLinearAccClamp=<10000 1000 1500>
    componentLinearAccScale=<0.5 0.3 0.5>
    componentLinearVelScale=<0.6 0.4 0.6>
    gravityOverride {z=-400}
    physicsBodyDefinitions [{
      collisionType=CustomSphere
      sphereCollisionRadius=1
      boundBone {boneName=dyn_strap_00}
      boxExtents=<14 14 12>
      constraintSetup {angularLimitsMax=<0 20 6>, angularLimitsMin=<0 -20 -2>, angularTarget=<0 0 0>}
      localJointOffset=<2 0 0>
     }]
   }
   n12 = Anim Dynamics - Bone: dyn_strap_01!(ComponentPose=@n11, Alpha=1, GravityOverride:Vector=@n8, ExternalForce="0,0,0") -> (Pose) {
    angularDampingOverride=0.9999
    angularSpringConstant=100
    bAngularSpring=true
    bOverrideAngularDamping=true
    bUseGravityOverride=true
    bUsePlanarLimit=false
    simulationSpace=RootRelative
    alphaBoolBlend {blendInTime=0.1, blendOption=Linear, blendOutTime=0.1}
    boundBone {boneName=dyn_strap_01}
    componentAppliedLinearAccClamp=<100 300 400>
    componentLinearAccScale=<0.5 0.1 0.5>
    componentLinearVelScale=<0.6 0.4 0.6>
    gravityOverride {z=-400}
    physicsBodyDefinitions [{
      collisionType=CustomSphere
      sphereCollisionRadius=2
      boundBone {boneName=dyn_strap_01}
      boxExtents=<13 13 13>
      constraintSetup {angularLimitsMax=<0 15 15>, angularLimitsMin=<0 -20 -25>, angularTarget=<0 0 0>}
      localJointOffset=<3 0 0>
     }]
    planarLimits [{drivingBone {boneName=dyn_pelvis}, planeTransform {location=<6.8641 -12.9029 1.0977>, rotation=<rot 17.9262 176.8852 47.4721>}}]
   }
   n13 = "Transform (Modify) Bone - Bone: pack"(
    ComponentPose=Local To Component(LocalPose=Use cached pose 'InputPose'())
    bAlphaBoolEnabled=Get Not in SkyDive() -> (Not in SkyDive:bool)
    Translation="0,0,4"
    Rotation="10,0,0"
   ) {
    alphaInputType=Bool
    rotationMode=BMM_Additive
    rotationSpace=BCS_BoneSpace
    scaleMode=BMM_Additive
    scaleSpace=BCS_BoneSpace
    translationMode=BMM_Additive
    translationSpace=BCS_BoneSpace
    alphaBoolBlend {blendOption=Linear}
    boneToModify {boneName=pack}
    rotation {pitch=10}
    translation {z=4}
   }
   n14 = "Transform (Modify) Bone - Bone: pack"(ComponentPose=@n13, bAlphaBoolEnabled=Get bIsSkyDiving(), Translation="0,-3,2", Rotation="10,0,10") {
    alphaInputType=Bool
    rotationMode=BMM_Additive
    rotationSpace=BCS_BoneSpace
    scaleMode=BMM_Additive
    scaleSpace=BCS_BoneSpace
    translationMode=BMM_Additive
    translationSpace=BCS_BoneSpace
    alphaBoolBlend {blendOption=Linear}
    boneToModify {boneName=pack}
    rotation {pitch=10, roll=10}
    translation {y=-3, z=2}
   }
   n15 = "Transform (Modify) Bone - Bone: dyn_pelvis"!(
    ComponentPose=@n14
    bAlphaBoolEnabled:bool=True
    Translation=Get Pelvice translation() -> (Pelvice translation:Vector)
    Rotation=Get Pelvice rotation() -> (Pelvice rotation:Rotator)
    Scale="1,1,1"
   ) -> (Pose) {
    alphaInputType=Bool
    rotationMode=BMM_Replace
    rotationSpace=BCS_WorldSpace
    translationMode=BMM_Replace
    translationSpace=BCS_WorldSpace
    alphaBoolBlend {blendInTime=0.4, blendOption=Linear, blendOutTime=0.4}
    boneToModify {boneName=dyn_pelvis}
   }
   n16 = Get Not In Frontend() -> (Not In Frontend:bool)
   n17 = Get GravityOverride()
   n18 = Anim Dynamics - Bone: dyn_strap_00!(ComponentPose=@n15, bAlphaBoolEnabled:bool=@n16, GravityOverride=@n17, ExternalForce="0,0,0") -> (Pose) {
    alphaInputType=Bool
    angularDampingOverride=0.9999
    angularSpringConstant=250
    bAngularSpring=true
    bOverrideAngularDamping=true
    bUseGravityOverride=true
    bUsePlanarLimit=false
    simulationSpace=RootRelative
    alphaBoolBlend {blendInTime=0.1, blendOption=Linear, blendOutTime=0.1}
    boundBone {boneName=dyn_strap_00}
    componentAppliedLinearAccClamp=<10000 1000 1500>
    componentLinearAccScale=<0.5 0.3 0.5>
    componentLinearVelScale=<0.6 0.4 0.6>
    gravityOverride {z=-400}
    physicsBodyDefinitions [{
      collisionType=CustomSphere
      sphereCollisionRadius=1
      boundBone {boneName=dyn_strap_00}
      boxExtents=<11 11 11>
      constraintSetup {angularLimitsMax=<0 10 0>, angularLimitsMin=<0 -20 -2>, angularTarget=<0 0 0>}
      localJointOffset=<4 0 0>
     }]
   }
   n19 = Anim Dynamics - Bone: dyn_strap_01!(ComponentPose=@n18, bAlphaBoolEnabled:bool=@n16, GravityOverride=@n17, ExternalForce="0,0,0") -> (Pose) {
    alphaInputType=Bool
    angularDampingOverride=0.9999
    angularSpringConstant=250
    bAngularSpring=true
    bOverrideAngularDamping=true
    bUseGravityOverride=true
    bUsePlanarLimit=false
    simulationSpace=RootRelative
    alphaBoolBlend {blendInTime=0.1, blendOption=Linear, blendOutTime=0.1}
    boundBone {boneName=dyn_strap_01}
    componentAppliedLinearAccClamp=<100 300 400>
    componentLinearAccScale=<0.5 0.1 0.5>
    componentLinearVelScale=<0.6 0.4 0.6>
    gravityOverride {z=-400}
    physicsBodyDefinitions [{
      collisionType=CustomSphere
      sphereCollisionRadius=2
      boundBone {boneName=dyn_strap_01}
      boxExtents=<12 12 11>
      constraintSetup {angularLimitsMax=<0 8 0>, angularLimitsMin=<0 -20 -25>, angularTarget=<0 0 0>}
      localJointOffset=<5.6 0 0>
     }]
    planarLimits [{drivingBone {boneName=dyn_pelvis}, planeTransform {location=<6.8641 -12.9029 1.0977>, rotation=<rot 17.9262 176.8852 47.4721>}}]
    sphericalLimits [{limitRadius=0, drivingBone {boneName=None}, sphereLocalOffset=<0 0 0>}]
   }
   n20 = Get JogOffset() -> (JogOffset:bool)
   n21 = "Transform (Modify) Bone - Bone: dyn_strap_01"!(ComponentPose=@n19, bAlphaBoolEnabled:bool=@n20, Translation="0,0,0", Rotation="0,-40,0", Scale="1,1,1") -> (Pose) {
    alphaInputType=Bool
    rotationMode=BMM_Additive
    rotationSpace=BCS_BoneSpace
    alphaBoolBlend {blendInTime=0.3, blendOption=Linear, blendOutTime=0.3}
    boneToModify {boneName=dyn_strap_01}
    rotation {yaw=-40}
   }
   n22 = "Transform (Modify) Bone - Bone: dyn_strap_00"!(ComponentPose=@n21, bAlphaBoolEnabled:bool=@n20, Translation="0,0,0", Rotation="0,-10,0", Scale="1,1,1") -> (Pose) {
    alphaInputType=Bool
    rotationMode=BMM_Additive
    rotationSpace=BCS_BoneSpace
    alphaBoolBlend {blendInTime=0.3, blendOption=Linear, blendOutTime=0.3}
    boneToModify {boneName=dyn_strap_00}
    rotation {yaw=-10}
   }
   n23 = "Transform (Modify) Bone - Bone: dyn_strap_01"(ComponentPose=@n22, bAlphaBoolEnabled=Get bIsCrouching(), Translation="0,0,0", Rotation="0,-15,0", Scale="1,1,1") {
    alphaInputType=Bool
    rotationMode=BMM_Additive
    rotationSpace=BCS_BoneSpace
    alphaBoolBlend {blendInTime=0.1, blendOption=Linear, blendOutTime=0.1}
    boneToModify {boneName=dyn_strap_01}
    rotation {yaw=-15}
   }
   n24 = "Transform (Modify) Bone - Bone: dyn_strap_01"(ComponentPose=@n23, bAlphaBoolEnabled=@n1, Translation="-2,-1.5,2.8", Rotation="0,0,0", Scale="1,1,1") {
    alphaInputType=Bool
    rotationMode=BMM_Additive
    rotationSpace=BCS_BoneSpace
    translationMode=BMM_Additive
    translationSpace=BCS_BoneSpace
    alphaBoolBlend {blendInTime=0.5, blendOption=Linear, blendOutTime=0.5}
    boneToModify {boneName=dyn_strap_01}
    translation=<-2 -1.5 2.8>
   }
   n25 = Get GravityOverride()
   n26 = Anim Dynamics - Bone: dyn_celery_01(ComponentPose=@n24, Alpha=1, GravityOverride=@n25, ExternalForce="0,0,0") {
    angularDampingOverride=0.9999
    angularSpringConstant=5000
    bAngularSpring=true
    bOverrideAngularDamping=true
    bUseGravityOverride=true
    bUsePlanarLimit=false
    gravityScale=0.35
    simulationSpace=BoneRelative
    alphaBoolBlend {blendOption=Linear}
    boundBone {boneName=dyn_celery_01}
    componentAppliedLinearAccClamp=<3500 3500 3500>
    componentLinearAccScale=<0.35 0.35 0.35>
    componentLinearVelScale=<0.35 0.35 0.35>
    gravityOverride {z=-700}
    physicsBodyDefinitions [{
      boundBone {boneName=dyn_celery_01}
      boxExtents=<20 20 20>
      constraintSetup {angularLimitsMax=<0 30 8>, angularLimitsMin=<0 -30 -8>, angularTarget=<1 0 0>}
      localJointOffset=<1 0 0>
     }]
    relativeSpaceBone {boneName=root}
   }
   n27 = Anim Dynamics - Bone: dyn_celery_02(ComponentPose=@n26, Alpha=1, GravityOverride=@n25, ExternalForce="0,0,0") {
    angularDampingOverride=0.9999
    angularSpringConstant=3000
    bAngularSpring=true
    bOverrideAngularDamping=true
    bUseGravityOverride=true
    bUsePlanarLimit=false
    gravityScale=0.35
    simulationSpace=BoneRelative
    alphaBoolBlend {blendOption=Linear}
    boundBone {boneName=dyn_celery_02}
    componentAppliedLinearAccClamp=<3500 3500 3500>
    componentLinearAccScale=<0.35 0.35 0.35>
    componentLinearVelScale=<0.35 0.35 0.35>
    gravityOverride {z=-700}
    physicsBodyDefinitions [{
      boundBone {boneName=dyn_celery_02}
      boxExtents=<20 20 20>
      constraintSetup {angularLimitsMax=<0 20 0>, angularLimitsMin=<0 -20 0>, angularTarget=<1 0 0>}
      localJointOffset=<2 0 0>
     }]
    relativeSpaceBone {boneName=root}
   }
   Output Pose(Result=Blend Poses by bool(BlendPose_0=Component To Local(ComponentPose=@n12), BlendPose_1=Component To Local(ComponentPose=@n27), bActiveValue=Get bIsInFrontEnd()) {blendType=Linear}) {layerGroup=None}
   Save cached pose 'InputPose'(Pose=Input Pose()) {cacheName=InputPose}
  ]}
 EventGraph {nodes [
   n1 = Event OnCharacterCustomizationLoaded()
   > n2 = Set Body Mesh(Body Mesh:Skeletal Mesh Component Object Reference=Get Array Item with EFort Custom Part Type(Array=@n1.CustomMeshParts)) -> (Output_Get:Skeletal Mesh Component Object Reference)
   Event BlueprintUpdateAnimation()
   > n3 = Sequence!(execute) -> (then_0, then_1, then_2:Exec)
   > n4 = Gravity Override State Function DEPRECATED(CharacterPartAnimInstance=Self-Reference(), GameTimeInSeconds=GetGameTimeInSeconds(), GravityOverrideParams=Get Gravity Override Params())
   > Set GravityOverride(GravityOverride=@n4)
   n5 = BreakTransform(InTransform=GetSocketTransform(self=@n2, InSocketName=spine_01, TransformSpace=RTS_World))
   > Set Pelvice translation(Pelvice translation:Vector=@n5) -> (Output_Get:Vector)
   > Set Pelvice rotation(Pelvice rotation:Rotator=@n5.Rotation) -> (Output_Get:Rotator)
   > Set Not In Frontend(Not In Frontend:bool=NOT Boolean(A=Get bIsInFrontEnd())) -> (Output_Get:bool)
   n6 = Get bIsTacticalSprinting()
   n7 = Get bIsCrouching()
   > Set JogOffset(JogOffset:bool=AND Boolean!(A=Get bIsSprinting(), B=NOT Boolean(A=@n7), C:bool=NOT Boolean(A=@n6)) -> (ReturnValue)) -> (Output_Get:bool)
   > Set Not in SkyDive(Not in SkyDive:bool=NOT Boolean(A=Get bIsSkyDiving())) -> (Output_Get:bool)
   n8 = Cast To FortPlayerPawn(execute=@n3.then_1, Object=TryGetPawnOwner())
   n9 = GetDisplayName(Object=GetSkeletalMeshForPartType(self=@n8))
   > Set JadeTowelHopeBody(JadeTowelHopeBody:bool=AND Boolean!(A=MatchesWildcard(SourceString=@n9, Wildcard=*JadeTowelHope*, SearchCase=CaseSensitive), B=Get bIsInFrontEnd()) -> (ReturnValue)) -> (Output_Get:bool)
   > Set JadeTowelGlossBody(JadeTowelGlossBody:bool=AND Boolean(A=MatchesWildcard(SourceString=@n9, Wildcard=*JadeTowelGloss*, SearchCase=CaseSensitive), B=Get bIsInFrontEnd())) -> (Output_Get:bool)
   > Set JadeTowelGlossBodyRotate(JadeTowelGlossBodyRotate:double=0) -> (Output_Get:double)
   n10 = Get bIsSkyDiving()
   n11 = Get bIsPlayingEmote()
   n12 = GetOwnerSkeletalMeshComponent()
   n13 = BreakRotator(InRot=BreakTransform(InTransform=GetDeltaTransformFromRefPose(self=@n12, BoneName=spine_05, BaseName=pelvis)).Rotation)
   n14 = Branch(
    execute=@n3.then_2
    Condition=AND Boolean!(
     A="float < float"(A=@n13.Yaw, B=-8)
     B=Get bIsInFrontEnd()
     C:bool=NOT Boolean(A=@n11)
     D:bool=NOT Boolean(A=@n10)
     E:bool=Get JadeTowelGlossBody()
    ) -> (ReturnValue)
   )
   > Set BodyAnimIdleFrontEnd_Z(BodyAnimIdleFrontEnd_Z:double=1) -> (Output_Get:double)
   Set BodyAnimIdleFrontEnd_Z(execute=@n14.else, BodyAnimIdleFrontEnd_Z=0)
  ]}
]
variables [
 Body Mesh {type=SkeletalMeshComponent}
 BodyAnimIdleFrontEnd_X {type=number}
 BodyAnimIdleFrontEnd_Z {type=number}
 Gravity Override Params {
  type=GravityOverrideParamsStruct
  default {
   emoteOrMelee {gravityOverride=<0 0 -600>}
   front_End {gravityOverride=<0 0 -600>}
   onGround_Falling {gravityOverride=<0 0 -600>}
   onGround_Moving {gravityOverride=<0 0 -600>}
   onGround_Standing {gravityOverride=<0 0 -600>}
   rail_Grinding {gravityOverride=<0 0 -600>}
   riding_Moving_Targeting {gravityOverride=<0 -200 -600>}
   riding_Standing {gravityOverride=<0 -350 -600>}
   skydive_FloatingDown {gravityOverride=<0 -800 600>, windAmplitude=<100 -300 400>, windFrequency=<30 50 10>, windOffset=<0 -700 500>}
   skydive_Surfing {gravityOverride=<500 -300 -600>}
   ziplining {gravityOverride=<0 0 -600>}
  }
 }
 GravityOverride {type=Vector}
 JadeTowelGlossBody {type=boolean}
 JadeTowelGlossBodyRotate {type=number}
 JadeTowelHopeBody {type=boolean}
 JogOffset {type=boolean}
 Not In Frontend {type=boolean}
 Not in SkyDive {type=boolean}
 Pelvice rotation {type=Rotator}
 Pelvice translation {type=Vector}
]
